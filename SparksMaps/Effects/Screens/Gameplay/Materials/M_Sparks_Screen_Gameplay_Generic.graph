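# Unlit material for gameplay screens in Sparks maps, drawing a noise-modulated gradient glow in the stage's colors.
# Two vertical gradients, each shaped by SmoothStep and Power parameters, are tinted by the Primary color from MF_MojoStageColors and summed, with an extra multiplier on layer B.
# The sum is masked by panning blurred noise, with tiling and speed packed into NoiseTile's R and B channels.
# MF_Sparks_Screen_Details receives the result as emissive.
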Material /SparksMaps/Effects/Screens/Gameplay/Materials/M_Sparks_Screen_Gameplay_Generic
graph {
 properties {
  bAutomaticallySetUsageInEditor=false
  bHasPixelAnimation=true
  bUsedWithNanite=true
  pixelDepthOffsetMode=PDOM_Legacy
  shadingModel=MSM_Unlit
  usedShadingModels=Unlit
 }
 nodes [
  n1 = VectorParameter() {group=ScreenTheme, parameterName=NoiseTile, defaultValue=<rgba 1 1 0 1>}
  n2 = Multiply(A=TextureCoordinate(), B=AppendVector(A=@n1.R, B=@n1.R))
  n3 = AppendVector(A=@n1.B, B=@n1.B)
  n4 = Add(
   A=TextureSample(UVs=Panner(Coordinate=@n2, Speed=@n3)) {samplerType=SAMPLERTYPE_Grayscale, texture=&LowResBlurredNoise}.R
   B=ScalarParameter() {group=ScreenTheme, parameterName=NoiseAdd}
  )
  n5 = SmoothStep(
   Min=ScalarParameter() {group=ScreenTheme, parameterName=SmoothStep Min B}
   Max=ScalarParameter() {defaultValue=1, group=ScreenTheme, parameterName=SmoothStep Max B}
   Value=MaterialFunctionCall() {materialFunction=&LinearGradient}.VGradient
  )
  n6 = ScalarParameter() {defaultValue=1, group=ScreenTheme, parameterName=Gradient Power B}
  n7 = MaterialFunctionCall!(CustomUPosition) {materialFunction=/SparksStageShow/Systems/StageManager/FXManager/MaterialFunctions/MF_MojoStageColors}
  n8 = Multiply(A=Multiply(A=Power(Base=@n5, Exp=@n6), B=@n7.Primary), B=ScalarParameter() {defaultValue=1, group=ScreenTheme, parameterName=Multiply B})
  n9 = ScalarParameter() {group=ScreenTheme, parameterName=SmoothStep Min A}
  n10 = ScalarParameter() {defaultValue=1, group=ScreenTheme, parameterName=SmoothStep Max A}
  n11 = MaterialFunctionCall() {materialFunction=&LinearGradient}
  n12 = Multiply(
   A=Power(
    Base=SmoothStep(Min=@n9, Max=@n10, Value=@n11.VGradient)
    Exp=ScalarParameter() {defaultValue=1, group=ScreenTheme, parameterName=Gradient Power A}
   )
   B=@n7.Primary
  )
  MaterialFunctionCall!(Emissive=Multiply(A=Saturate(None=@n4), B=Add(A=@n8, B=@n12)), Scanlines Tile) {materialFunction=/SparksMaps/Effects/Screens/Materials/MF/MF_Sparks_Screen_Details}
 ]
}
